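; Gameplay ability for the Season 24 narrative quest sensor backpack, playing VFX when the player passes through a ring.
; On the server it waits for a ring-interact tag on the pawn, calls RingPassed, then ends after a short delay if the quest is completed.
; RingPassed, gated on cosmetic execution, attaches a Niagara effect to the backpack mesh's idle FX socket.
; Ending the ability removes effects granting the sensor-pack equipped tag.
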(event Ability|EventActivateAbility
  ;[0 0 0] Activate ability - wait for the player to pass through a ring (server execution)
  (|CallParentFunction)
  (if (Networking|IsServer)
    (Ability|Tasks|WaitGameplayTagAddToActor (CallFunction|GetAvatarasFortPlayerPawn) (Variables|Default|GetRingInteractTag)
      (:Added
        (CallFunction|RingPassed)
        (bind _bcompletedquest (CallFunction|HasPlayerCompletedQuest))
        (if _bcompletedquest
          (Development|PrintString "Quest completed" true true "(R=1.000000,G=0.000000,B=0.562280,A=1.000000)" 4.0)
          (Utilities|FlowControl|Delay 1.0)
          (Ability|EndAbility))))))

(event Custom|RingPassed
  ;[0 0 0] spawn attach
  (Networking|CanExecuteCosmeticEvents
    (:True
      ;;[0.08854 0.08854 0.08854] VFX Hook/Niagara spawn fx
      (Development|PrintString "Backpack VFX triggered" true true "(R=1.000000,G=0.000000,B=0.562280,A=1.000000)" 4.0)
      (bind _asfort_player_pawn_athena (Utilities|Casting|CastToFortPlayerPawnAthena (Ability|GetAvatarActorfromActorInfo)))
      (if (Utilities|IsValid (FortPawn|GetModifierforPartType _asfort_player_pawn_athena "Backpack") :!class /Script/Engine.KismetSystemLibrary)
        (Niagara|SpawnSystemAttached "/Game/Animation/Props/Gadgets_Backbling_Detector/Skin/Detector_Ch4S2/FX/NS_Quest_S24_Narrative_Q3_BackSensor.NS_Quest_S24_Narrative_Q3_BackSensor" (CustomCharacter|GetSkeletalMeshForPartOnPlayerPawn (FortPawn|GetModifierforPartType _asfort_player_pawn_athena "Backpack")) (Class|CustomCharacterPartModifier|GetIdleFXSocketName (FortPawn|GetModifierforPartType _asfort_player_pawn_athena "Backpack")) "0.000000,-15.000000,-6.000000")))
    (:False
      (Development|PrintString "Cannot execute cosmetic events" true true "(R=1.000000,G=0.000000,B=0.562280,A=1.000000)" 4.0))))

(event Ability|EventOnEndAbility
  (params (bWasCancelled :type Boolean))
  (|CallParentFunction bWasCancelled)
  (Development|PrintString "Pack ability ending..." true true "(R=1.000000,G=0.000000,B=0.562280,A=1.000000)" 4.0)
  (GameplayEffects|RemoveActiveEffectswithGrantedTags (Ability|GetFortAbilitySystemComponent (Ability|GetAvatarActorfromActorInfo)) "(GameplayTags=((TagName=\"Granted.Quest.S24.SensorPack.Equipped\")))"))

(fn HasPlayerCompletedQuest
  (return (Quest|HasCompletedQuest (Quests|GetQuestManager (Fortnite|GetFortPlayerControllerfromActor (CallFunction|GetAvatarasFortPlayerPawn)) "Invalid") (Variables|Default|GetQuestDefinition) :!class /Script/FortniteGame.FortQuestManager)))
